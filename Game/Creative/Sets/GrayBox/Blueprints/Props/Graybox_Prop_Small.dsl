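; Small graybox prop Blueprint for Creative that exposes its texture data assets so they can be referenced in UEFN.
; It overrides the exposed asset provider functions, calling the parent in each and adding version 1 to the provider version.
; Reference collection loops over the building texture data, resolves each soft reference and registers it as an indirect asset reference.
; The construction script is empty.
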(fn BuildExposedAssetProviderVersion(Implementation)
  (params (Version :type Fort%20Exposed%20Asset%20Provider%20Version%20Structure%20%28by%20ref%29))
  (|CallParentFunction Version)
  (ExposedAssetProvider|AddVersion Version 1)
  (return))

(fn CollectAssetReferencestoAllow(Implementation)
  (params (Collector :type Fort%20Exposed%20Asset%20Reference%20Collector%20Structure%20%28by%20ref%29))
  (|CallParentFunction Collector)
  (bind _outtexturedataarray (Building|GetBuildingTextureData))
  (for _array_element _outtexturedataarray
    (ExposedAssetProvider|CollectIndirectAssetReference Collector (Utilities|ResolveSoftReference _array_element)) ; Collect the texture data to allow it to be referenced in UEFN
  )
  (return))

(fn ConstructionScript)
